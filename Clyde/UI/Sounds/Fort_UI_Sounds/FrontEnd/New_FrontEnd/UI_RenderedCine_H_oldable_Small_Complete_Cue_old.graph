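# Sound cue, marked old, that plays the completion sound for a small holdable button in rendered-cinematic front-end UI.
# A mixer layers a quieted hold-complete press, an unmodified deep-impact shop click and a boosted, pitched-down cinematic-skip whoosh.
# The cue is routed to the rendered cinematics sound class at reduced volume, with a concurrency override that blocks new instances past a small limit.
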SoundCue /Clyde/UI/Sounds/Fort_UI_Sounds/FrontEnd/New_FrontEnd/UI_RenderedCine_H_oldable_Small_Complete_Cue_old
root=Mixer_0
nodes [
 {id=Mixer_0, children [Modulator_0,Modulator_1,WavePlayer_2], properties {inputVolume [1,1,1]}}
 {id=Modulator_0, children [WavePlayer_0], properties {volumeMax=0.7, volumeMin=0.7}}
 {id=Modulator_1, children [WavePlayer_4], properties {pitchMax=0.8, pitchMin=0.8, volumeMax=1.5, volumeMin=1.5}}
 {
  id=WavePlayer_0
  properties {soundWaveAssetPtr=/Game/Sounds/Fort_UI_Sounds/FrontEnd/DefaultSounds/HoldableButtons/wav/UI_Button_Press_Hold_Quiet_Complete}
 }
 {id=WavePlayer_2, properties {soundWaveAssetPtr=/Game/Sounds/Fort_UI_Sounds/FrontEnd/ItemShop/wav/Fort_UI_Shop_Click_DeepImpact_01}}
 {id=WavePlayer_4, properties {soundWaveAssetPtr=/Game/Sounds/Fort_UI_Sounds/FrontEnd/UI_Cinematics_Skip_Button_Pressed_Whoosh}}
]
properties {
 bOverrideConcurrency=true
 maxDistance=2097152
 soundClassObject=/Game/Sounds/FortSoundClassesAndModes/RenderedCinematics
 volumeMultiplier=0.378
 concurrencyOverrides {maxCount=2, resolutionRule=PreventNew, retriggerTime=0.5}
}
