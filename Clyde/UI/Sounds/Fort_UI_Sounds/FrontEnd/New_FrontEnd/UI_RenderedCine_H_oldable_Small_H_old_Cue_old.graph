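# Sound cue for a looping press-and-hold UI button sound, a quiet "small hold" variant in the RenderedCinematics sound class.
# A single looping wave player feeds a modulator that raises pitch by a random amount within a narrow range.
# The cue plays at reduced volume and overrides concurrency to limit simultaneous instances and rapid retriggers.
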SoundCue /Clyde/UI/Sounds/Fort_UI_Sounds/FrontEnd/New_FrontEnd/UI_RenderedCine_H_oldable_Small_H_old_Cue_old
root=Modulator_1
nodes [
 {id=Modulator_1, children [WavePlayer_1], properties {pitchMax=1.4, pitchMin=1.3, volumeMin=1}}
 {
  id=WavePlayer_1
  properties {bLooping=true, soundWaveAssetPtr=/Game/Sounds/Fort_UI_Sounds/FrontEnd/DefaultSounds/HoldableButtons/wav/UI_Button_Press_Hold_Quiet_Hold_lp}
 }
]
properties {
 bOverrideConcurrency=true
 maxDistance=2097152
 soundClassObject=/Game/Sounds/FortSoundClassesAndModes/RenderedCinematics
 volumeMultiplier=0.189
 concurrencyOverrides {maxCount=2, retriggerTime=0.2}
}
