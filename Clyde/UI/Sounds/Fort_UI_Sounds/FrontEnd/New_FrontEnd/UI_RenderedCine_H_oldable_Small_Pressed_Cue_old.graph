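# Sound cue for the press of a small holdable front-end UI button, routed to the RenderedCinematics sound class and suffixed as an old version.
# A single quiet press-and-hold wave plays through a Modulator left at its default settings, with the cue volume turned down.
# The cue overrides concurrency with a short retrigger time.
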SoundCue /Clyde/UI/Sounds/Fort_UI_Sounds/FrontEnd/New_FrontEnd/UI_RenderedCine_H_oldable_Small_Pressed_Cue_old
root=Modulator_2
nodes [
 {id=Modulator_2, children [WavePlayer_1]}
 {
  id=WavePlayer_1
  properties {soundWaveAssetPtr=/Game/Sounds/Fort_UI_Sounds/FrontEnd/DefaultSounds/HoldableButtons/wav/UI_Button_Press_Hold_Quiet_Press}
 }
]
properties {
 bOverrideConcurrency=true
 maxDistance=2097152
 soundClassObject=/Game/Sounds/FortSoundClassesAndModes/RenderedCinematics
 volumeMultiplier=0.378
 concurrencyOverrides {retriggerTime=0.15}
}
